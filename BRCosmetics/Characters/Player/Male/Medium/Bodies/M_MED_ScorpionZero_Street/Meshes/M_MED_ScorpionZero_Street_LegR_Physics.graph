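# Physics asset for the right leg mesh of the male medium ScorpionZero Street body, covering the dynamic pants bones around the calf.
# A calf capsule anchors a master pants bone, which branches into four sub-masters, each leading a two-bone chain.
# Master bones carry tiny capsules and free-rotating joints, acting as pivots.
# Chain bones share identical fat, offset capsules and narrow swing and twist limits, keeping cloth-like motion tight.
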PhysicsAsset /BRCosmetics/Characters/Player/Male/Medium/Bodies/M_MED_ScorpionZero_Street/Meshes/M_MED_ScorpionZero_Street_LegR_Physics
bodies {
 calf_r [{shapeType=Capsule, length=12, radius=6, shapeName=calf_r_capsule, center=<8 2 0>, rotation=<rot 90 0 0>}]
 dyn_pants_r_01 [{shapeType=Capsule, length=8, radius=10, shapeName=dyn_pants_r_01_capsule, center=<3 0 4>, rotation=<rot 90 0 0>}]
 dyn_pants_r_01_MSTR [{shapeType=Capsule, length=0.505, radius=0.505, shapeName=dyn_pants_r_01_MSTR_capsule}]
 dyn_pants_r_01_end [{shapeType=Capsule, length=8, radius=10, shapeName=dyn_pants_r_01_end_capsule, center=<3 0 4>, rotation=<rot 90 0 0>}]
 dyn_pants_r_02 [{shapeType=Capsule, length=8, radius=10, shapeName=dyn_pants_r_02_capsule, center=<3 0 4>, rotation=<rot 90 0 0>}]
 dyn_pants_r_02_MSTR [{shapeType=Capsule, length=0.505, radius=0.505, shapeName=dyn_pants_r_02_MSTR_capsule}]
 dyn_pants_r_02_end [{shapeType=Capsule, length=8, radius=10, shapeName=dyn_pants_r_02_end_capsule, center=<3 0 4>, rotation=<rot 90 0 0>}]
 dyn_pants_r_03 [{shapeType=Capsule, length=8, radius=10, shapeName=dyn_pants_r_03_capsule, center=<3 0 4>, rotation=<rot 90 0 0>}]
 dyn_pants_r_03_MSTR [{shapeType=Capsule, length=0.505, radius=0.505, shapeName=dyn_pants_r_03_MSTR_capsule}]
 dyn_pants_r_03_end [{shapeType=Capsule, length=8, radius=10, shapeName=dyn_pants_r_03_end_capsule, center=<3 0 4>, rotation=<rot 90 0 0>}]
 dyn_pants_r_04 [{shapeType=Capsule, length=8, radius=10, shapeName=dyn_pants_r_04_capsule, center=<3 0 4>, rotation=<rot 90 0 0>}]
 dyn_pants_r_04_MSTR [{shapeType=Capsule, length=0.505, radius=0.505, shapeName=dyn_pants_r_04_MSTR_capsule}]
 dyn_pants_r_04_end [{shapeType=Capsule, length=8, radius=10, shapeName=dyn_pants_r_04_end_capsule, center=<3 0 4>, rotation=<rot 90 0 0>}]
 dyn_pants_r_MSTR [{shapeType=Capsule, length=0.505, radius=0.505, shapeName=dyn_pants_r_MSTR_capsule, rotation=<rot 0 -90 0>}]
}
constraints [
 {
  bone1Name=dyn_pants_r_MSTR
  bone2Name=calf_r
  swing1LimitDegrees=45
  swing1Motion=Free
  swing2LimitDegrees=45
  swing2Motion=Free
  twistLimitDegrees=45
  twistMotion=Free
 }
 {bone1Name=dyn_pants_r_01, bone2Name=dyn_pants_r_01_MSTR, swing1LimitDegrees=10, swing2LimitDegrees=4, twistLimitDegrees=2}
 {bone1Name=dyn_pants_r_01_end, bone2Name=dyn_pants_r_01, swing1LimitDegrees=10, swing2LimitDegrees=4, twistLimitDegrees=2}
 {bone1Name=dyn_pants_r_02, bone2Name=dyn_pants_r_02_MSTR, swing1LimitDegrees=10, swing2LimitDegrees=4, twistLimitDegrees=2}
 {bone1Name=dyn_pants_r_02_end, bone2Name=dyn_pants_r_02, swing1LimitDegrees=10, swing2LimitDegrees=4, twistLimitDegrees=2}
 {bone1Name=dyn_pants_r_03, bone2Name=dyn_pants_r_03_MSTR, swing1LimitDegrees=10, swing2LimitDegrees=4, twistLimitDegrees=2}
 {bone1Name=dyn_pants_r_03_end, bone2Name=dyn_pants_r_03, swing1LimitDegrees=10, swing2LimitDegrees=4, twistLimitDegrees=2}
 {bone1Name=dyn_pants_r_04, bone2Name=dyn_pants_r_04_MSTR, swing1LimitDegrees=10, swing2LimitDegrees=4, twistLimitDegrees=2}
 {bone1Name=dyn_pants_r_04_end, bone2Name=dyn_pants_r_04, swing1LimitDegrees=10, swing2LimitDegrees=4, twistLimitDegrees=2}
 {
  bone1Name=dyn_pants_r_04_MSTR
  bone2Name=dyn_pants_r_MSTR
  swing1LimitDegrees=45
  swing1Motion=Free
  swing2LimitDegrees=45
  swing2Motion=Free
  twistLimitDegrees=45
  twistMotion=Free
 }
 {
  bone1Name=dyn_pants_r_03_MSTR
  bone2Name=dyn_pants_r_MSTR
  swing1LimitDegrees=45
  swing1Motion=Free
  swing2LimitDegrees=45
  swing2Motion=Free
  twistLimitDegrees=45
  twistMotion=Free
 }
 {
  bone1Name=dyn_pants_r_02_MSTR
  bone2Name=dyn_pants_r_MSTR
  swing1LimitDegrees=45
  swing1Motion=Free
  swing2LimitDegrees=45
  swing2Motion=Free
  twistLimitDegrees=45
  twistMotion=Free
 }
 {
  bone1Name=dyn_pants_r_01_MSTR
  bone2Name=dyn_pants_r_MSTR
  swing1LimitDegrees=45
  swing1Motion=Free
  swing2LimitDegrees=45
  swing2Motion=Free
  twistLimitDegrees=45
  twistMotion=Free
 }
]
