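# Physics asset for the M_MED_ScorpionZero_Street male body cosmetic, covering its dynamic dagger and pouch accessory bones.
# Capsule bodies sit on the pelvis, each accessory bone and a tiny master (MSTR) bone per accessory.
# Each accessory is tightly angle-limited to its master, with narrow twist and uneven swing limits.
# Both masters are constrained to the pelvis with free swing and twist.
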PhysicsAsset /BRCosmetics/Characters/Player/Male/Medium/Bodies/M_MED_ScorpionZero_Street/Meshes/M_MED_ScorpionZero_Street_Physics
bodies {
 dyn_dagger [{shapeType=Capsule, length=15, radius=5, shapeName=dyn_dagger_capsule, center=<3 0 -3>, rotation=<rot 90 0 0>}]
 dyn_dagger_MSTR [{shapeType=Capsule, length=0.505, radius=0.505, shapeName=dyn_dagger_MSTR_capsule, rotation=<rot 0 -90 0>}]
 dyn_pouch [{shapeType=Capsule, length=3, radius=5, shapeName=dyn_pouch_capsule, center=<3 0 0>}]
 dyn_pouch_MSTR [{shapeType=Capsule, length=0.505, radius=0.505, shapeName=dyn_pouch_MSTR_capsule, rotation=<rot 0 -90 0>}]
 pelvis [{shapeType=Capsule, length=3.5, radius=11.9263, shapeName=pelvis_capsule}]
}
constraints [
 {bone1Name=dyn_pouch, bone2Name=dyn_pouch_MSTR, swing1LimitDegrees=4, swing2LimitDegrees=16, twistLimitDegrees=2}
 {bone1Name=dyn_dagger, bone2Name=dyn_dagger_MSTR, swing1LimitDegrees=16, swing2LimitDegrees=8, twistLimitDegrees=2}
 {
  bone1Name=dyn_pouch_MSTR
  bone2Name=pelvis
  swing1LimitDegrees=45
  swing1Motion=Free
  swing2LimitDegrees=45
  swing2Motion=Free
  twistLimitDegrees=45
  twistMotion=Free
 }
 {
  bone1Name=dyn_dagger_MSTR
  bone2Name=pelvis
  swing1LimitDegrees=45
  swing1Motion=Free
  swing2LimitDegrees=45
  swing2Motion=Free
  twistLimitDegrees=45
  twistMotion=Free
 }
]
