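# Post-process style cosmetic AnimBP for the ScorpionZero Street male body, adding secondary motion to the pose copied from the parent mesh.
# A Control Rig (bypassed while skydiving) feeds pouch and dagger constraints, chain AnimDynamics and pelvis RigidBody simulation.
# Per-calf RigidBody sims switch between standing and crouching setups, with additive pants offsets when crouching, jumping or falling.
# The EventGraph updates gravity override and WPO cloth materials, including foot velocity and skydive-float wind.
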AnimBlueprint /BRCosmetics/Characters/Player/Male/Medium/Bodies/M_MED_ScorpionZero_Street/Meshes/M_MED_ScorpionZero_Street_AnimBP
target_skeleton=&/M_MED_ScorpionZero_Street_Skeleton
graphs [
 AnimGraph {nodes [
   Layered blend per bone(BlendPoses_0=Slot 'DefaultSlot'(), BlendWeights_0=Get Custom Idle Is Playing() -> (Custom Idle Is Playing:float)) {layerSetup [{branchFilters [{boneName=dyn_collar_l_MSTR}, {boneName=dyn_collar_l_f_MSTR}, {boneName=dyn_collar_r_MSTR}, {boneName=dyn_collar_r_f_MSTR}]}]}
   n1 = RigidBody!(ComponentPose=Local To Component(LocalPose=Use cached pose 'SavedPose8934'()), Alpha, OverrideWorldGravity:Vector="0,0,0", ExternalForce) -> (Pose) {
    bOverrideWorldGravity=true
    overridePhysicsAsset=&/M_MED_ScorpionZero_Street_Stand_LegL_Physics
    simulationSpace=BaseBoneSpace
    alphaBoolBlend {blendOption=Linear}
    baseBoneRef {boneName=calf_l}
    componentAppliedLinearAccClamp=<3000 3000 3000>
    componentLinearAccScale=<0.4 0.4 0.5>
    componentLinearVelScale=<0.15 0.15 0.25>
    simSpaceSettings {
     maxAngularAcceleration=100
     maxAngularVelocity=3
     maxLinearAcceleration=700
     maxLinearVelocity=500
     velocityScaleZ=0
     worldAlpha=1
     externalLinearDragV=<0.2 0.2>
    }
   }
   n2 = RigidBody(ComponentPose=@n1, OverrideWorldGravity="0,0,0") {
    bOverrideWorldGravity=true
    overridePhysicsAsset=&/M_MED_ScorpionZero_Street_Stand_LegR_Physics
    simulationSpace=BaseBoneSpace
    alphaBoolBlend {blendOption=Linear}
    baseBoneRef {boneName=calf_r}
    componentAppliedLinearAccClamp=<3000 3000 3000>
    componentLinearAccScale=<0.4 0.4 0.5>
    componentLinearVelScale=<0.15 0.15 0.25>
    simSpaceSettings {
     maxAngularAcceleration=100
     maxAngularVelocity=3
     maxLinearAcceleration=700
     maxLinearVelocity=500
     velocityScaleZ=0
     worldAlpha=1
     externalLinearDragV=<0.2 0.2>
    }
   }
   n3 = Component To Local(ComponentPose=@n2)
   n4 = RigidBody(ComponentPose=Local To Component(LocalPose=Use cached pose 'SavedPose8934'()), OverrideWorldGravity="0,0,0") {
    bOverrideWorldGravity=true
    overridePhysicsAsset=&/M_MED_ScorpionZero_Street_LegL_Physics
    simulationSpace=BaseBoneSpace
    alphaBoolBlend {blendOption=Linear}
    baseBoneRef {boneName=calf_l}
    componentAppliedLinearAccClamp=<3000 3000 3000>
    componentLinearAccScale=<0.4 0.4 0.5>
    componentLinearVelScale=<0.15 0.15 0.25>
    simSpaceSettings {
     maxAngularAcceleration=100
     maxAngularVelocity=3
     maxLinearAcceleration=700
     maxLinearVelocity=500
     velocityScaleZ=0
     worldAlpha=1
     externalLinearDragV=<0.2 0.2>
    }
   }
   n5 = RigidBody(ComponentPose=@n4, OverrideWorldGravity="0,0,0") {
    bOverrideWorldGravity=true
    overridePhysicsAsset=&/M_MED_ScorpionZero_Street_LegR_Physics
    simulationSpace=BaseBoneSpace
    alphaBoolBlend {blendOption=Linear}
    baseBoneRef {boneName=calf_r}
    componentAppliedLinearAccClamp=<3000 3000 3000>
    componentLinearAccScale=<0.4 0.4 0.5>
    componentLinearVelScale=<0.15 0.15 0.25>
    simSpaceSettings {
     maxAngularAcceleration=100
     maxAngularVelocity=3
     maxLinearAcceleration=700
     maxLinearVelocity=500
     velocityScaleZ=0
     worldAlpha=1
     externalLinearDragV=<0.2 0.2>
    }
   }
   n6 = "Transform (Modify) Bone - Bone: dyn_pants_r_03_MSTR"(ComponentPose=@n5, Alpha=1, Translation="0,-3,-2.5", Rotation="0,0,5", Scale="1,1,1") {
    rotationMode=BMM_Additive
    scaleMode=BMM_Replace
    translationMode=BMM_Additive
    alphaBoolBlend {blendOption=Linear}
    boneToModify {boneName=dyn_pants_r_03_MSTR}
    rotation {roll=5}
    translation {y=-3, z=-2.5}
   }
   n7 = "Transform (Modify) Bone - Bone: dyn_pants_l_03_MSTR"(ComponentPose=@n6, Alpha=1, Translation="0,-3,-2.5", Rotation="0,0,5", Scale="1,1,1") {
    rotationMode=BMM_Additive
    translationMode=BMM_Additive
    alphaBoolBlend {blendOption=Linear}
    boneToModify {boneName=dyn_pants_l_03_MSTR}
    rotation {roll=5}
    translation {y=-3, z=-2.5}
   }
   n8 = Component To Local(ComponentPose=@n7)
   n9 = Get bIsCrouching()
   n10 = "To Float (Boolean)"(InBool=OR Boolean(A=Get bIsJumping(), B=Get bIsFalling()))
   n11 = "Transform (Modify) Bone - Bone: dyn_pants_r_03_MSTR"(
    ComponentPose=Local To Component(LocalPose=Blend Poses by bool(BlendPose_0=@n8, BlendPose_1=@n3, bActiveValue=@n9) {blendType=Linear})
    Alpha=@n10
    Translation="0,-1,-4"
    Rotation="0,0,5"
    Scale="1,1,1"
   ) {
    rotationMode=BMM_Additive
    translationMode=BMM_Additive
    alphaBoolBlend {blendOption=Linear}
    alphaScaleBiasClamp {bInterpResult=true}
    boneToModify {boneName=dyn_pants_r_03_MSTR}
    rotation {roll=5}
    translation {y=-1, z=-4}
   }
   n12 = "Transform (Modify) Bone - Bone: dyn_pants_l_03_MSTR"(ComponentPose=@n11, Alpha=@n10, Translation="0,-1,-4", Rotation="0,0,7", Scale="1,1,1") {
    rotationMode=BMM_Additive
    translationMode=BMM_Additive
    alphaBoolBlend {blendOption=Linear}
    alphaScaleBiasClamp {bInterpResult=true}
    boneToModify {boneName=dyn_pants_l_03_MSTR}
    rotation {roll=7}
    translation {y=-1, z=-4}
   }
   Output Pose(Result=Component To Local(ComponentPose=@n12)) {layerGroup=None}
   n13 = Control Rig(Source=Use cached pose 'SavedPose18800'()) {
    alphaBoolBlend {blendOption=Linear}
    controlRigAssetReference {blueprintRigClass=&/M_MED_ScorpionZero_Street_CtrlRig.M_MED_ScorpionZero_Street_CtrlRig_C}
   }
   n14 = Use cached pose 'SavedPose18800'()
   n15 = Get bIsSkyDiving()
   n16 = "Constraint - dyn_pouch_MSTR (P)"(
    ComponentPose=Local To Component(LocalPose=Blend Poses by bool(BlendPose_0=@n14, BlendPose_1=@n13, bActiveValue=@n15) {blendType=Linear})
    Alpha=0.2
    ConstraintWeights_0=1
   ) {
    alphaBoolBlend {blendOption=Linear}
    boneToModify {boneName=dyn_pouch_MSTR}
    constraintSetup [{targetBone {boneName=spine_01}}]
    constraintWeights [1]
   }
   n17 = "Constraint - dyn_dagger_MSTR (P)"(ComponentPose=@n16, Alpha=0.2, ConstraintWeights_0=1) {
    alphaBoolBlend {blendOption=Linear}
    boneToModify {boneName=dyn_dagger_MSTR}
    constraintSetup [{targetBone {boneName=spine_01}}]
    constraintWeights [1]
   }
   n18 = Anim Dynamics - Bone: dyn_chain(ComponentPose=@n17, Alpha=1, ExternalForce="0,0,0") {
    alphaBoolBlend {blendOption=Linear}
    boundBone {boneName=dyn_chain}
    physicsBodyDefinitions [{
      boundBone {boneName=dyn_chain}
      boxExtents=<10 10 20>
      constraintSetup {angularLimitsMax=<1 3 3>, angularLimitsMin=<-1 -3 -12>, angularTarget=<0 0 0>}
      localJointOffset=<2 0 0>
     }]
   }
   n19 = RigidBody(ComponentPose=@n18, OverrideWorldGravity="0,0,0") {
    bOverrideWorldGravity=true
    overridePhysicsAsset=&/M_MED_ScorpionZero_Street_Physics
    simulationSpace=BaseBoneSpace
    alphaBoolBlend {blendOption=Linear}
    baseBoneRef {boneName=pelvis}
    componentAppliedLinearAccClamp=<3000 3000 3000>
    componentLinearAccScale=<0.25 0.2 0.4>
    componentLinearVelScale=<0.1 0.1 0.2>
    simSpaceSettings {
     maxAngularAcceleration=100
     maxAngularVelocity=3
     maxLinearAcceleration=700
     maxLinearVelocity=500
     velocityScaleZ=0
     worldAlpha=1
     externalLinearDragV=<0.2 0.2>
    }
   }
   Save cached pose 'SavedPose8934'(Pose=Component To Local(ComponentPose=@n19)) {cacheName=SavedPose8934}
   Save cached pose 'SavedPose18800'(Pose=Copy Pose From Mesh() {bUseAttachedParent=true}) {cacheName=SavedPose18800}
  ]}
 EventGraph {nodes [
   Event BlueprintUpdateAnimation()
   > n1 = Sequence!(execute) -> (then_0, then_1, then_2:Exec, then_3:Exec)
   > n2 = Gravity Override State Function DEPRECATED(CharacterPartAnimInstance=Self-Reference(), GameTimeInSeconds=GetGameTimeInSeconds(), GravityOverrideParams=Get Gravity Override Params())
   > Set Gravity Override(Gravity Override=@n2)
   n3 = Set SkelMesh(execute=@n1.then_1, SkelMesh=GetPartSkeletalMeshComponent())
   > n4 = Get Dynamic Material Instance(SkeletalMeshComponent=@n3)
   > Set Mat Instance Dynamic(Mat Instance Dynamic=@n4)
   > n5 = GetOrCreateDynamicOverlayMaterial(MeshComponent=@n3, MaterialName=None)
   > Set Overlay Mat Instance Dynamic(Overlay Mat Instance Dynamic=@n5)
   Set WPO Globals Function(
    execute=@n1.then_2
    Material=Get Mat Instance Dynamic()
    Overlay Material=Get Overlay Mat Instance Dynamic()
    Skeletal Mesh=Get SkelMesh()
    "Local Noise Transform  Location"=Get Local Noise Transform Location() -> (Local Noise Transform Location:Vector)
    Local Noise Offset=Get Local Noise Offset()
   )
   > Set WPO Regional Function!(
    execute
    Material:Material Instance Dynamic Object Reference=Get Mat Instance Dynamic()
    Overlay Material:Material Instance Dynamic Object Reference=Get Overlay Mat Instance Dynamic()
    Skeletal Mesh:Skeletal Mesh Component Object Reference=Get SkelMesh()
    Velocity Transform:Name=foot_r
    Velocity Interp Speed:double=Get Velocity Interp Speed()
    Collision Transform:Name=calf_r
    Collision Transform Axis:EAxis Enum
    Invert Collision Transform Axis:bool=false
    Delta Time X:double=0
    Previous Velocity Transform Location:"Vector (by ref)"=Get R1 Velocity Transform Location()
    Previous Velocity Transform Velocity:"Vector (by ref)"=Get R1 Velocity Transform Velocity()
    Region:int=1
   ) -> (then)
   > Set WPO Regional Function(
    Material=Get Mat Instance Dynamic()
    Overlay Material=Get Overlay Mat Instance Dynamic()
    Skeletal Mesh=Get SkelMesh()
    Velocity Transform=foot_l
    Velocity Interp Speed=Get Velocity Interp Speed()
    Collision Transform=calf_l
    Invert Collision Transform Axis=true
    Delta Time X=0
    Previous Velocity Transform Location=Get R2 Velocity Transform Location()
    Previous Velocity Transform Velocity=Get R2 Velocity Transform Velocity()
    Region=2
   )
   > n6 = Branch(Condition=Get bIsSkydiveFloating())
   > n7 = Set WPO Pho Wind Intensity(WPO Pho Wind Intensity=1)
   SetScalarParameterValue(
    execute=[@n8.then, @n7.then]
    self=[Get Mat Instance Dynamic(), Get Overlay Mat Instance Dynamic()]
    ParameterName=ClothSim_WindStrength
    Value=Get WPO Pho Wind Intensity()
   )
   n8 = Set WPO Pho Wind Intensity(execute=@n6.else, WPO Pho Wind Intensity=0)
   Branch(execute=@n1.then_3, Condition=Get bIsInFrontEnd())
   > Set Custom Idle Is Playing(Custom Idle Is Playing:float=GetCurveValue(self=GetOwnerFrontEndAnimInstance(), CurveName=Frontend_Idle_Bypass)) -> (Output_Get:float)
   Event BlueprintInitializeAnimation()
   > Set SkelMesh(SkelMesh=GetOwnerSkeletalMeshComponent())
   n9 = TryGetPawnOwner()
   > Is Valid(InputObject=@n9)
   > n10 = Cast To FortPlayerPawnAthena(Object=@n9)
   > Set PlayerPawn(PlayerPawn:Fort Player Pawn Athena Object Reference=@n10) -> (Output_Get:Fort Player Pawn Athena Object Reference)
  ]}
]
variables [
 Custom Idle Is Playing {type=number}
 Gravity Override {type=Vector}
 Gravity Override Params {type=GravityOverrideParamsStruct}
 Local Noise Offset {type=Vector}
 Local Noise Transform Location {type=Vector}
 Mat Instance Dynamic {type=MaterialInstanceDynamic}
 Overlay Mat Instance Dynamic {type=MaterialInstanceDynamic}
 PlayerPawn {type=FortPlayerPawnAthena}
 R1 Velocity Transform Location {type=Vector}
 R1 Velocity Transform Velocity {type=Vector}
 R2 Velocity Transform Location {type=Vector}
 R2 Velocity Transform Velocity {type=Vector}
 SkelMesh {type=SkeletalMeshComponent}
 Velocity Interp Speed {type=number}
 WPO Pho Wind Intensity {type=number}
]
