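# Control Rig for the Scorpion Zero Street character's face accessory that adjusts the dynamic headband master bone as the neck moves.
# A spherical pose reader on neck_01 measures its orientation within an asymmetric active and falloff region.
# Its output weights a Modify Transforms node that applies a fixed rotation and translation offset to dyn_headband_MSTR.
# The left and right headband chains follow only as children of that bone.
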ControlRigBlueprint /BRCosmetics/Characters/Player/Male/Medium/Bodies/M_MED_ScorpionZero_Street/Meshes/Parts/M_MED_ScorpionZero_Street_FaceAcc_CtrlRig
graphs [
 RigVMFunctionLibrary {links [], nodes []}
 RigVMModel {nodes [
   RigUnit_BeginExecution = &RigUnit_BeginExecution()
   n1 = &RigUnit_SphericalPoseReader(
    ActiveRegionScaleFactors=(PositiveWidth=3.5,NegativeWidth=2,PositiveHeight=3,NegativeHeight=3)
    ActiveRegionSize=0.1
    Debug=(bDrawDebug=true,bDraw2D=false,bDrawLocalAxes=false,DebugScale=25,DebugSegments=20,DebugThickness=0.25)
    DriverAxis=(X=1,Y=0,Z=0)
    DriverItem=(Type=Bone,Name="neck_01")
    FalloffRegionScaleFactors=(PositiveWidth=1,NegativeWidth=0.6,PositiveHeight=1,NegativeHeight=1)
    FalloffSize=0.45
    FlipHeightScaling=true
    FlipWidthScaling=true
    OptionalParentItem=(Type=Bone,Name="None")
    RotationOffset=(X=-120,Y=0,Z=30)
    ExecutePin=@RigUnit_BeginExecution.ExecutePin
   )
   RigVMFunction_MathQuaternionFromEuler = &RigVMFunction_MathQuaternionFromEuler(Euler=(X=0,Y=0,Z=-60))
   > &RigUnit_ModifyTransforms(
    ItemToModify=((Item=(Type=Bone,Name="dyn_headband_MSTR"),Transform=(Rotation=(X=0,Y=0,Z=0,W=1),Translation=(X=0,Y=-2.5,Z=0),Scale3D=(X=1,Y=1,Z=1))))
    WeightMinimum=0
    ItemToModify.0.Transform.Rotation=@RigVMFunction_MathQuaternionFromEuler.Result
    Weight=@n1.OutputParam
   )
  ]}
]
hierarchy [
 root {loc=<0 0 0>, rot=<0 0 0 1>}
 pelvis {parent=root, loc=<0 -0.7071 93.1191>, rot=<0.0127 0.707 -0.0127 -0.707>}
 spine_01 {parent=pelvis, loc=<6.7494 -0.6388 0>, rot=<0 0 0.034 -0.9994>}
 spine_02 {parent=spine_01, loc=<7.2518 -0.035 0>, rot=<0 0 0.0238 -0.9997>}
 spine_03 {parent=spine_02, loc=<6.0104 0 0>, rot=<0 0 0.0149 -0.9999>}
 spine_04 {parent=spine_03, loc=<8.1024 0 0>, rot=<0 0 -0.0132 -0.9999>}
 spine_05 {parent=spine_04, loc=<10.0146 0 0>, rot=<0 0 -0.0015 -1>}
 neck_01 {parent=spine_05, loc=<11.3138 0.1529 0>, rot=<0 0 -0.2003 -0.9797>}
 neck_02 {parent=neck_01, loc=<4.5627 0 0>, rot=<0 0 0.0504 -0.9987>}
 head {parent=neck_02, loc=<5.3914 0 0>, rot=<0 0 0.115 -0.9934>}
 dyn_headband_MSTR {parent=head, loc=<5.4611 -8.0531 -0.0689>, rot=<0.5934 -0.8049 0 0>}
 dyn_headband_l_01 {parent=dyn_headband_MSTR, loc=<1.0665 1.2982 1.1937>, rot=<-0.0612 0.0493 -0.36 -0.9296>}
 dyn_headband_l_02 {parent=dyn_headband_l_01, loc=<6.3606 0 0>, rot=<0.0066 -0.0155 -0.1181 -0.9929>}
 dyn_headband_l_03 {parent=dyn_headband_l_02, loc=<5.816 0 0>, rot=<0.0006 -0.0019 0.0375 -0.9993>}
 dyn_headband_l_04 {parent=dyn_headband_l_03, loc=<5.6226 0 0>, rot=<0.002 -0.0043 0.0734 -0.9973>}
 dyn_headband_r_01 {parent=dyn_headband_MSTR, loc=<1.2632 1.4648 -1.3786>, rot=<0.0701 -0.0575 -0.3678 -0.9255>}
 dyn_headband_r_02 {parent=dyn_headband_r_01, loc=<6.15 0 0>, rot=<-0.0041 0.0096 -0.0996 -0.995>}
 dyn_headband_r_03 {parent=dyn_headband_r_02, loc=<5.9023 0 0>, rot=<0.0009 -0.0026 0.029 -0.9996>}
 dyn_headband_r_04 {parent=dyn_headband_r_03, loc=<5.8126 0 0>, rot=<-0.0061 0.0132 0.0724 -0.9973>}
]
