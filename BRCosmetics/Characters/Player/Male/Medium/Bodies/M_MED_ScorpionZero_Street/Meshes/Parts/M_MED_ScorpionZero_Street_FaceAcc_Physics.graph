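# Physics asset for the ScorpionZero Street male face accessory, providing bodies for a dynamic headband with left and right strands.
# Two four-bone capsule chains branch from the headband master bone, linked by constraints with moderate swing and very tight twist limits.
# Larger capsules on the neck, upper spine and upper arms approximate the torso, apparently as collision for the strands.
# Constraints among these torso bodies, and from headband master to neck, leave angular motion free.
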PhysicsAsset /BRCosmetics/Characters/Player/Male/Medium/Bodies/M_MED_ScorpionZero_Street/Meshes/Parts/M_MED_ScorpionZero_Street_FaceAcc_Physics
bodies {
 dyn_headband_MSTR [{
   shapeType=Capsule
   length=2.177
   radius=2.0809
   shapeName=dyn_headband_MSTR_capsule
   center=<-0.2162 -0.2486 0.0073>
   rotation=<rot -41.0446 -90 -9.1972>
  }]
 dyn_headband_l_01 [{shapeType=Capsule, length=3, radius=2.5, shapeName=dyn_headband_l_01_capsule, center=<0 -1 0>, rotation=<rot 90 0 0>}]
 dyn_headband_l_02 [{shapeType=Capsule, length=3, radius=2.5, shapeName=dyn_headband_l_02_capsule, center=<0 -1 0>, rotation=<rot 90 0 0>}]
 dyn_headband_l_03 [{shapeType=Capsule, length=3, radius=2.5, shapeName=dyn_headband_l_03_capsule, center=<0 -1 0>, rotation=<rot 90 0 0>}]
 dyn_headband_l_04 [{shapeType=Capsule, length=3, radius=2.5, shapeName=dyn_headband_l_04_capsule, center=<0 -1 0>, rotation=<rot 90 0 0>}]
 dyn_headband_r_01 [{shapeType=Capsule, length=3, radius=2.5, shapeName=dyn_headband_r_01_capsule, center=<0 -1 0>, rotation=<rot 90 0 0>}]
 dyn_headband_r_02 [{shapeType=Capsule, length=3, radius=2.5, shapeName=dyn_headband_r_02_capsule, center=<0 -1 0>, rotation=<rot 90 0 0>}]
 dyn_headband_r_03 [{shapeType=Capsule, length=3, radius=2.5, shapeName=dyn_headband_r_03_capsule, center=<0 -1 0>, rotation=<rot 90 0 0>}]
 dyn_headband_r_04 [{shapeType=Capsule, length=3, radius=2.5, shapeName=dyn_headband_r_04_capsule, center=<0 -1 0>, rotation=<rot 90 0 0>}]
 neck_01 [{shapeType=Capsule, length=4, radius=10.5, shapeName=neck_01_capsule, center=<3.2889 1.6316 0>}]
 spine_04 [{shapeType=Capsule, length=18.5, radius=13, shapeName=spine_05_capsule, center=<0.9545 2.4246 0>, rotation=<rot 0 -160.0006 -180>}]
 spine_05 [
  {shapeType=Capsule, length=18.5, radius=13, shapeName=spine_05_capsule, center=<1.0019 1.8422 0>, rotation=<rot 0 -160.0006 -180>}
  {shapeType=Capsule, length=2.5, radius=10.5, shapeName=neck_01_capsule, center=<11.9124 2.8781 0>, rotation=<rot 90 -179.9984 -179.9981>}
 ]
 upperarm_l [{shapeType=Capsule, length=10, radius=8, shapeName=upperarm_l_capsule, center=<4.0823 -4.3107 -0.0348>, rotation=<rot 90 0 0>}]
 upperarm_r [{shapeType=Capsule, length=10, radius=8, shapeName=upperarm_r_capsule, center=<-4.0823 4.3107 0.0348>, rotation=<rot 90 0 0>}]
}
constraints [
 {
  bone1Name=dyn_headband_MSTR
  bone2Name=neck_01
  swing1LimitDegrees=45
  swing1Motion=Free
  swing2LimitDegrees=45
  swing2Motion=Free
  twistLimitDegrees=45
  twistMotion=Free
 }
 {bone1Name=dyn_headband_l_01, bone2Name=dyn_headband_MSTR, swing1LimitDegrees=30, swing2LimitDegrees=30, twistLimitDegrees=3}
 {bone1Name=dyn_headband_l_02, bone2Name=dyn_headband_l_01, swing1LimitDegrees=30, swing2LimitDegrees=30, twistLimitDegrees=3}
 {bone1Name=dyn_headband_l_03, bone2Name=dyn_headband_l_02, swing1LimitDegrees=30, swing2LimitDegrees=30, twistLimitDegrees=3}
 {bone1Name=dyn_headband_l_04, bone2Name=dyn_headband_l_03, swing1LimitDegrees=30, swing2LimitDegrees=30, twistLimitDegrees=3}
 {bone1Name=dyn_headband_r_01, bone2Name=dyn_headband_MSTR, swing1LimitDegrees=30, swing2LimitDegrees=30, twistLimitDegrees=3}
 {bone1Name=dyn_headband_r_02, bone2Name=dyn_headband_r_01, swing1LimitDegrees=30, swing2LimitDegrees=30, twistLimitDegrees=3}
 {bone1Name=dyn_headband_r_03, bone2Name=dyn_headband_r_02, swing1LimitDegrees=30, swing2LimitDegrees=30, twistLimitDegrees=3}
 {bone1Name=dyn_headband_r_04, bone2Name=dyn_headband_r_03, swing1LimitDegrees=30, swing2LimitDegrees=30, twistLimitDegrees=3}
 {
  bone1Name=neck_01
  bone2Name=spine_05
  swing1LimitDegrees=45
  swing1Motion=Free
  swing2LimitDegrees=45
  swing2Motion=Free
  twistLimitDegrees=45
  twistMotion=Free
 }
 {
  bone1Name=upperarm_l
  bone2Name=spine_05
  swing1LimitDegrees=45
  swing1Motion=Free
  swing2LimitDegrees=45
  swing2Motion=Free
  twistLimitDegrees=45
  twistMotion=Free
 }
 {
  bone1Name=upperarm_r
  bone2Name=spine_05
  swing1LimitDegrees=45
  swing1Motion=Free
  swing2LimitDegrees=45
  swing2Motion=Free
  twistLimitDegrees=45
  twistMotion=Free
 }
 {
  bone1Name=spine_05
  bone2Name=spine_04
  swing1LimitDegrees=45
  swing1Motion=Free
  swing2LimitDegrees=45
  swing2Motion=Free
  twistLimitDegrees=45
  twistMotion=Free
 }
]
